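# Material function for the FeatherMudGlance character body cosmetic that outputs animated, flickering lightning tinted by an HDR Lightning Color parameter.
# Panning tiling noise distorts separately panned UVs for samples of a lightning flipbook-named texture, drawing on its R, G and B channels.
# A SphereMask on a second panned noise, with radius driven by a time-based sine, breaks up and pulses the combined bolts.
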MaterialFunction /BRCosmetics/Characters/Player/Male/Medium/Bodies/M_MED_FeatherMudGlance/Materials/MF_FeatherMudGlance_Lightning
graph {expressions {
  MaterialExpressionAdd_0 {
   class=MaterialExpressionAdd
   inputs [
    A {from=MaterialExpressionSaturate_0}
    B {from=MaterialExpressionTextureSampleParameter2D_5, output=B}
   ]
  }
  MaterialExpressionAdd_10 {
   class=MaterialExpressionAdd
   inputs [
    A {from=MaterialExpressionReroute_1}
    B {from=MaterialExpressionPanner_8}
   ]
  }
  MaterialExpressionAdd_11 {
   class=MaterialExpressionAdd
   inputs [
    A {from=MaterialExpressionReroute_0}
    B {from=MaterialExpressionPanner_9}
   ]
  }
  MaterialExpressionAdd_7 {
   class=MaterialExpressionAdd
   inputs [
    A {from=MaterialExpressionMultiply_24}
    B {from=MaterialExpressionPanner_6}
   ]
  }
  MaterialExpressionConstantBiasScale_0 {class=MaterialExpressionConstantBiasScale, inputs [None {from=MaterialExpressionSine_0}]}
  MaterialExpressionConstant_0 {class=MaterialExpressionConstant, properties {r=0.5}}
  MaterialExpressionConstant_1 {class=MaterialExpressionConstant, properties {r=0.5}}
  MaterialExpressionConstant_24 {class=MaterialExpressionConstant, properties {r=0.5}}
  MaterialExpressionFunctionOutput_0 {class=MaterialExpressionFunctionOutput, inputs [None {from=MaterialExpressionMultiply_28, output=""}], properties {outputName=Lightning Out}}
  MaterialExpressionMax_0 {
   class=MaterialExpressionMax
   inputs [
    A {from=MaterialExpressionSaturate_0}
    B {from=MaterialExpressionPower_1}
   ]
  }
  MaterialExpressionMax_2 {
   class=MaterialExpressionMax
   inputs [
    A {from=MaterialExpressionPower_1}
    B {from=MaterialExpressionPower_0}
   ]
  }
  MaterialExpressionMultiply_0 {
   class=MaterialExpressionMultiply
   inputs [
    A {from=MaterialExpressionTextureSampleParameter2D_4, output=G}
    B {from=""}
   ]
   properties {constB=0.5}
  }
  MaterialExpressionMultiply_15 {
   class=MaterialExpressionMultiply
   inputs [
    A {from=MaterialExpressionTime_0}
    B {from=MaterialExpressionConstant_0}
   ]
  }
  MaterialExpressionMultiply_16 {
   class=MaterialExpressionMultiply
   inputs [
    A {from=MaterialExpressionSine_0}
    B {from=MaterialExpressionScalarParameter_19}
   ]
   properties {constB=0.8}
  }
  MaterialExpressionMultiply_17 {
   class=MaterialExpressionMultiply
   inputs [
    A {from=MaterialExpressionPanner_4}
    B {from=MaterialExpressionConstant_1}
   ]
  }
  MaterialExpressionMultiply_22 {
   class=MaterialExpressionMultiply
   inputs [
    A {from=MaterialExpressionPanner_5}
    B {from=""}
   ]
   properties {constB=2}
  }
  MaterialExpressionMultiply_23 {
   class=MaterialExpressionMultiply
   inputs [
    A {from=MaterialExpressionSphereMask_2}
    B {from=MaterialExpressionMax_2}
   ]
  }
  MaterialExpressionMultiply_24 {
   class=MaterialExpressionMultiply
   inputs [
    A {from=MaterialExpressionTextureSample_15, output=R}
    B {from=""}
   ]
   properties {constB=0.2}
  }
  MaterialExpressionMultiply_25 {
   class=MaterialExpressionMultiply
   inputs [
    A {from=MaterialExpressionAdd_10}
    B {from=""}
   ]
  }
  MaterialExpressionMultiply_26 {
   class=MaterialExpressionMultiply
   inputs [
    A {from=MaterialExpressionAdd_11}
    B {from=""}
   ]
   properties {constB=3}
  }
  MaterialExpressionMultiply_27 {
   class=MaterialExpressionMultiply
   inputs [
    A {from=MaterialExpressionAdd_7}
    B {from=""}
   ]
  }
  MaterialExpressionMultiply_28 {
   class=MaterialExpressionMultiply
   inputs [
    A {from=MaterialExpressionPower_5}
    B {from=MaterialExpressionVectorParameter_10, output=RGB}
   ]
  }
  MaterialExpressionPanner_4 {
   class=MaterialExpressionPanner
   inputs [
    Coordinate {from=MaterialExpressionTextureCoordinate_2}
    Time {from=""}
    Speed {from=""}
   ]
   properties {speedX=0.05, speedY=0.1}
  }
  MaterialExpressionPanner_5 {
   class=MaterialExpressionPanner
   inputs [
    Coordinate {from=MaterialExpressionTextureCoordinate_5}
    Time {from=""}
    Speed {from=""}
   ]
   properties {speedX=0.02, speedY=0.1}
  }
  MaterialExpressionPanner_6 {
   class=MaterialExpressionPanner
   inputs [
    Coordinate {from=MaterialExpressionTextureCoordinate_3}
    Time {from=""}
    Speed {from=""}
   ]
   properties {speedX=-0.1, speedY=0.2}
  }
  MaterialExpressionPanner_8 {
   class=MaterialExpressionPanner
   inputs [
    Coordinate {from=MaterialExpressionTextureCoordinate_7}
    Time {from=""}
    Speed {from=""}
   ]
   properties {speedX=0.02, speedY=0.3}
  }
  MaterialExpressionPanner_9 {
   class=MaterialExpressionPanner
   inputs [
    Coordinate {from=MaterialExpressionTextureCoordinate_8}
    Time {from=""}
    Speed {from=""}
   ]
   properties {speedX=0.01, speedY=0.15}
  }
  MaterialExpressionPower_0 {
   class=MaterialExpressionPower
   inputs [
    Base {from=MaterialExpressionTextureSampleParameter2D_3, output=R}
    Exp {from=""}
   ]
   properties {constExponent=3}
  }
  MaterialExpressionPower_5 {
   class=MaterialExpressionPower
   inputs [
    Base {from=MaterialExpressionMultiply_23}
    Exp {from=""}
   ]
   properties {constExponent=1}
  }
  MaterialExpressionReroute_0 {class=MaterialExpressionReroute, inputs [None {from=MaterialExpressionMultiply_24}]}
  MaterialExpressionReroute_1 {class=MaterialExpressionReroute, inputs [None {from=MaterialExpressionMultiply_24}]}
  MaterialExpressionSaturate_0 {class=MaterialExpressionSaturate, inputs [None {from=MaterialExpressionMultiply_0}]}
  MaterialExpressionScalarParameter_19 {class=MaterialExpressionScalarParameter, properties {defaultValue=0.2, group=Lightning, parameterName=Radius}}
  MaterialExpressionScalarParameter_22 {class=MaterialExpressionScalarParameter, properties {defaultValue=0.7, group=Lightning, parameterName=Hardness}}
  MaterialExpressionSine_0 {class=MaterialExpressionSine, inputs [None {from=MaterialExpressionMultiply_15}], properties {period=0.1}}
  MaterialExpressionSphereMask_2 {
   class=MaterialExpressionSphereMask
   inputs [
    A {from=MaterialExpressionTextureSample_16, output=R}
    B {from=MaterialExpressionConstant_24, output=""}
    Radius {from=MaterialExpressionMultiply_16, output=""}
    Hardness {from=MaterialExpressionScalarParameter_22, output=""}
   ]
   properties {}
  }
  MaterialExpressionTextureCoordinate_2 {class=MaterialExpressionTextureCoordinate}
  MaterialExpressionTextureCoordinate_3 {class=MaterialExpressionTextureCoordinate}
  MaterialExpressionTextureCoordinate_5 {class=MaterialExpressionTextureCoordinate}
  MaterialExpressionTextureCoordinate_7 {class=MaterialExpressionTextureCoordinate}
  MaterialExpressionTextureCoordinate_8 {class=MaterialExpressionTextureCoordinate}
  MaterialExpressionTextureSampleParameter2D_3 {
   class=MaterialExpressionTextureSampleParameter2D
   inputs [
    UVs {from=MaterialExpressionMultiply_25, output=""}
    Apply View MipBias {from="", output=""}
   ]
   properties {
    group=Lightning
    parameterName=Lightning Texture
    texture=/BRCosmetics/Characters/Player/Male/Medium/Bodies/M_MED_FeatherMudGlance/Textures/LightningFlip_T2
   }
  }
  MaterialExpressionTextureSampleParameter2D_4 {
   class=MaterialExpressionTextureSampleParameter2D
   inputs [
    UVs {from=MaterialExpressionMultiply_26, output=""}
    Apply View MipBias {from="", output=""}
   ]
   properties {
    group=Lightning
    parameterName=Lightning Texture
    texture=&LightningFlip_T2
   }
  }
  MaterialExpressionTextureSampleParameter2D_5 {
   class=MaterialExpressionTextureSampleParameter2D
   inputs [
    UVs {from=MaterialExpressionMultiply_27, output=""}
    Apply View MipBias {from="", output=""}
   ]
   properties {
    group=Lightning
    parameterName=Lightning Texture2
    texture=&LightningFlip_T2
   }
  }
  MaterialExpressionTextureSample_15 {
   class=MaterialExpressionTextureSample
   inputs [
    UVs {from=MaterialExpressionMultiply_17}
    Tex {from=""}
    Apply View MipBias {from=""}
   ]
   properties {samplerSource=SSM_Wrap_WorldGroupSettings, texture=/Game/Global/Textures/Default/Noise/T_Tiling_Noise_01}
  }
  MaterialExpressionTextureSample_16 {
   class=MaterialExpressionTextureSample
   inputs [
    UVs {from=MaterialExpressionMultiply_22}
    Tex {from=""}
    Apply View MipBias {from=""}
   ]
   properties {samplerSource=SSM_Wrap_WorldGroupSettings, texture=&T_Tiling_Noise_01}
  }
  MaterialExpressionTime_0 {class=MaterialExpressionTime}
  MaterialExpressionVectorParameter_10 {class=MaterialExpressionVectorParameter, properties {group=Lightning, parameterName=Lightning Color, defaultValue=<rgba 25 8.242 24.1763 1>}}
 }}
