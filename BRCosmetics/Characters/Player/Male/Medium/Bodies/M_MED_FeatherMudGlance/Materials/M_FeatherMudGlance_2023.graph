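# Character body material for the FeatherMudGlance cosmetic, adding animated, reactivity-scaled glow effects to the shared 2023 character shading.
# It inserts an emissive override between the CharacterParent PreFX and PostFX functions, built from lightning, body edge, pants, and shoes-and-logo layers.
# E2 texture channels and Fresnel mask the layers, a pre-skinned-position sphere mask shapes the waist, and panning noise supplies smoke and color variation.
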Material /BRCosmetics/Characters/Player/Male/Medium/Bodies/M_MED_FeatherMudGlance/Materials/M_FeatherMudGlance_2023
graph {
 properties {
  bUseMaterialAttributes=true
  bUsedWithMeshDeformer=true
  bUsedWithMorphTargets=true
  bUsedWithNiagaraMeshParticles=true
  bUsedWithSkeletalMesh=true
  pixelDepthOffsetMode=PDOM_Legacy
  shadingModel=MSM_FromMaterialExpression
  usedShadingModels=Default Lit | Subsurface | Clear Coat | Subsurface Profile | Cloth
 }
 nodes [
  n1 = MaterialFunctionCall() {materialFunction=&MF_CharacterParent_2023_PreFX}
  MaterialFunctionCall(
   Material=SetMaterialAttributes(MaterialAttributes=@n1.Result, Emissive Color=NamedRerouteUsage()) {attributeSetTypes [#1]}
   M=@n1.M
   Pre Skinned Local Position=@n1.M
   Distance Blend=@n1.M
   Fresnel=@n1.M
  ) {materialFunction=&MF_CharacterParent_2023_PostFX}
  n2 = MaterialFunctionCall(
   UVs=Multiply(A=TextureCoordinate(), B=ScalarParameter() {defaultValue=5, parameterName=Paramed}) {constB=3}
   "Rotation Angle (0-1)"=Constant()
  ) {materialFunction=&CustomRotator}
  n3 = TextureSample(UVs=Panner(Coordinate=@n2.Rotated Values) {speedY=0.1}) {samplerSource=SSM_Wrap_WorldGroupSettings, samplerType=SAMPLERTYPE_Grayscale, texture=&LowResBlurredNoise}
  n4 = TextureSample(UVs=Panner(Coordinate=@n2.Rotated Values) {speedY=0.5}) {samplerSource=SSM_Wrap_WorldGroupSettings, samplerType=SAMPLERTYPE_Grayscale, texture=&LowResBlurredNoise}
  n5 = Add(A=Panner(Coordinate=@n2.Rotated Values) {speedY=0.2}, B=Multiply(A=@n4.R) {constB=0.5})
  NamedRerouteDeclaration(None=Multiply(
    A=LinearInterpolate(Alpha=@n3.R) {constA=0.6}
    B=TextureSample(UVs=@n5) {samplerSource=SSM_Wrap_WorldGroupSettings, texture=/Game/Global/Textures/Default/Noise/T_Tiling_Noise_01}.R
   )) {name=Smoke, nodeColor=<rgba 0 0.8828 1 1>}
  n6 = TextureSample(UVs=Panner(Coordinate=TextureCoordinate()) {speedY=0.2}) {samplerSource=SSM_Wrap_WorldGroupSettings, samplerType=SAMPLERTYPE_Grayscale, texture=&LowResBlurredNoise}
  NamedRerouteDeclaration(None=Power(Base=@n6.R)) {name=ColorNoise, nodeColor=<rgba 1 0 0.7031 1>}
  n7 = Power(Base=Fresnel(ExponentIn=Constant() {r=1.5}, BaseReflectFractionIn=Constant() {r=0.04}))
  n8 = TextureSample() {
   samplerSource=SSM_Wrap_WorldGroupSettings
   texture=/BRCosmetics/Characters/Player/Male/Medium/Bodies/M_MED_FeatherMudGlance/Textures/T_FeatherMudGlace_Body_E2
  }
  n9 = Multiply(A=@n8.G) {constB=10}
  n10 = Max(A=Max(A=Multiply(A=@n7) {constB=2}, B=Saturate(None=@n9)), B=@n8.B)
  n11 = NamedRerouteUsage()
  NamedRerouteDeclaration(None=Multiply(
    A=MaterialFunctionCall() {materialFunction=&/MF_FeatherMudGlance_Lightning}.Lightning Out
    B=Multiply(A=@n10, B=@n11)
   )) {name=Lightning, nodeColor=<rgba 1 0.9141 0 1>}
  NamedRerouteDeclaration(None=@n1.Pre skinned Local Position) {name=Pre skinned Local IN, nodeColor=<rgba 0.3828 0 1 1>}
  n12 = LinearInterpolate(A=Constant3Vector() {constant=<rgba 0.1556 0 4 1>}, B=Constant3Vector() {constant=<rgba 3.1589 0.4372 6 1>}, Alpha=NamedRerouteUsage())
  n13 = Constant() {r=2}
  n14 = Constant() {r=1.5}
  n15 = TextureSample() {
   samplerSource=SSM_Wrap_WorldGroupSettings
   texture=&T_FeatherMudGlace_Body_E2
  }
  n16 = Power(Base=Fresnel(ExponentIn=LinearInterpolate(A=@n13, B=@n14, Alpha=@n15.R), BaseReflectFractionIn=Constant() {r=0.04}))
  n17 = Multiply(A=Constant3Vector() {constant=<rgba 40 4 32.4937 1>}, B=@n16)
  n18 = TextureSample() {
   samplerSource=SSM_Wrap_WorldGroupSettings
   texture=&T_FeatherMudGlace_Body_E2
  }
  n19 = OneMinus(None=SphereMask(A=NamedRerouteUsage(), B=Constant3Vector() {constant=<rgba 0 0 80 1>}, Radius=Constant() {r=35}, Hardness=Constant()))
  n20 = Constant() {r=3}
  n21 = NamedRerouteDeclaration(None=Saturate(None=Power(Base=@n19, Exp=@n20) {constExponent=5})) {name=Waist mask, nodeColor=<rgba 1 0 0.5391 1>}
  NamedRerouteDeclaration(None=LinearInterpolate(B=LinearInterpolate(A=@n12, B=@n17, Alpha=@n18.R), Alpha=Multiply(A=@n16, B=@n21))) {name=Body edge glow, nodeColor=<rgba 1 0.4453 0 1>}
  n22 = TextureSample() {
   samplerSource=SSM_Wrap_WorldGroupSettings
   texture=&T_FeatherMudGlace_Body_E2
  }
  n23 = Power(Base=@n22.B)
  NamedRerouteDeclaration(None=Multiply(A=Multiply(A=@n23) {constB=1.5}, B=Constant3Vector() {constant=<rgba 10.3819 0 20 1>})) {name=Shoes and logo, nodeColor=<rgba 0.1046 0.1046 0.1046 1>}
  n24 = TextureSample() {
   samplerSource=SSM_Wrap_WorldGroupSettings
   texture=&T_FeatherMudGlace_Body_E2
  }
  n25 = NamedRerouteUsage()
  n26 = Constant() {r=1.5}
  n27 = Multiply(A=Power(Base=@n24.G) {constExponent=1.2}, B=Multiply(A=Power(Base=@n25, Exp=@n26), B=Constant() {r=8}) {constB=20})
  n28 = Multiply(A=@n24.G, B=Constant() {r=1.5}) {constB=0.3}
  n29 = Constant3Vector() {constant=<rgba 3.7681 0 10 1>}
  n30 = Constant3Vector() {constant=<rgba 5.191 0 10 1>}
  n31 = NamedRerouteUsage()
  NamedRerouteDeclaration(None=Multiply(A=Add(A=@n27, B=@n28), B=LinearInterpolate(A=@n29, B=@n30, Alpha=@n31))) {name=Pants glow, nodeColor=<rgba 1 0 0.6797 1>}
  n32 = NamedRerouteUsage()
  n33 = NamedRerouteUsage()
  n34 = Add(A=NamedRerouteUsage(), B=Add(A=NamedRerouteUsage(), B=Add(A=@n32, B=@n33)))
  n35 = ScalarParameter() {parameterName=reactivity}
  NamedRerouteDeclaration(None=Multiply(A=Multiply(A=@n34, B=@n35), B=ScalarParameter() {defaultValue=1, parameterName=reactivity_toggle})) {name=EmissiveOut, nodeColor=<rgba 0 1 1 1>}
 ]
}
