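# Translucent, unlit, two-sided glow material for the FeatherMudGlance character body, flagged for Niagara sprite and mesh particles.
# Color A, particle color or a panning rainbow gradient blends toward Color B through scrolling world-space noise and a local-bounds vertical gradient, feeding an emissive adjustment function.
# An alpha mask multiplies particle alpha, a depth-faded soft outline function using custom depth, invertible Fresnel and the powered vertical mask.
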Material /BRCosmetics/Characters/Player/Male/Medium/Bodies/M_MED_FeatherMudGlance/Materials/M_FeatherMudGlance_BodyGlow
graph {
 properties {
  bAutomaticallySetUsageInEditor=false
  bUsedWithNiagaraMeshParticles=true
  bUsedWithNiagaraSprites=true
  blendMode=BLEND_Translucent
  shadingModel=MSM_Unlit
  twoSided=true
  usedShadingModels=Unlit
 }
 nodes [
  n1 = LinearInterpolate(
   A=VectorParameter() {parameterName=Color A, defaultValue=<rgba 0 1 0 1>}.RGB
   B=ParticleColor().RGB
   Alpha=ScalarParameter() {parameterName=UseParticleColor}
  )
  n2 = TextureCoordinate()
  n3 = ScalarParameter() {parameterName=UVPanTime}
  n4 = AppendVector(
   A=ScalarParameter() {group=Distortion, parameterName=Texture Speed X, sliderMax=2, sliderMin=-2}
   B=ScalarParameter() {defaultValue=0.5, group=Distortion, parameterName=Texture Speed Y, sliderMax=2, sliderMin=-2}
  )
  n5 = ScalarParameter() {defaultValue=1, group=Distortion, parameterName=Texture Tiling U, sliderMax=2}
  n6 = ScalarParameter() {defaultValue=1, group=Distortion, parameterName=Texture Tiling V, sliderMax=2}
  n7 = TextureSampleParameter2D(UVs=Multiply(A=Panner(Coordinate=@n2, Time=@n3, Speed=@n4) {speedX=-1}, B=AppendVector(A=@n5, B=@n6)) {constB=2}) {parameterName=Panning Texture, texture=/Game/Effects/Fort_Effects/Textures/Grad/T_Gradient_Rainbow_NoRed}
  n8 = ScalarParameter() {parameterName=UsePanningTextureSample?}
  n9 = TransformPosition(None=WorldPosition() {worldPositionShaderOffset=WPT_ExcludeAllShaderOffsets}.XYZ) {transformSourceType=TRANSFORMPOSSOURCE_World}
  n10 = MaterialFunctionCall(Float3=@n9) {materialFunction=&BreakOutFloat3Components}
  n11 = Multiply(A=Time(), B=ScalarParameter() {defaultValue=-15, parameterName=Outline Noise Scroll Speed}) {constB=-15}
  n12 = Noise!(
   World Position=MaterialFunctionCall(X=@n10.R, Y=@n10.R, Z=Add(A=@n10.B, B=@n11)) {materialFunction=&MakeFloat3}.Result
   FilterWidth
  ) {bTurbulence=false, levels=3, outputMin=0, scale=0.025}
  n13 = Power(Base=@n12)
  n14 = Add(A=@n9, B=VectorParameter() {parameterName=Vertical Mask Offset, defaultValue=<rgba 0 0 0 1>}.RGB)
  n15 = MaterialFunctionCall() {materialFunction=&ObjectLocalBounds}
  n16 = MaterialFunctionCall(Float3=Divide(A=Subtract(A=@n14, B=@n15.Local Bounds Minimum), B=@n15.Local Bounds Size)) {materialFunction=&BreakOutFloat3Components}
  n17 = StaticSwitchParameter(True=@n16.B, False=OneMinus(None=@n16.B)) {parameterName=Invert Vertical Mask}
  n18 = Multiply(A=@n17, B=ScalarParameter() {defaultValue=1.5, parameterName=Gradient Mult}) {constB=2}
  MaterialFunctionCall(Emissive=LinearInterpolate(
    A=LinearInterpolate(A=@n1, B=@n7.RGB, Alpha=@n8)
    B=VectorParameter() {parameterName=Color B, defaultValue=<rgba 1 0 0 1>}.RGB
    Alpha=Multiply(A=@n13, B=@n18)
   )) {materialFunction=&MF_FX_EmissiveAdjustments}
  n19 = ParticleColor()
  n20 = DepthFade(
   Opacity=MaterialFunctionCall!(
    GlowMeshStencilIndex=ScalarParameter() {defaultValue=40, group=Soft Outline, parameterName="GlowMeshStencilIndex (S)"}
    UseCustomDepth=StaticBoolParameter() {defaultValue=true, group=Soft Outline, parameterName=UseCustomDepth}
    UseCustomStencil=StaticBoolParameter() {group=Soft Outline, parameterName=UseCustomStencil}
    OutlineThickness=ScalarParameter() {defaultValue=2.7, group=Soft Outline, parameterName="OutlineThickness (S)"}
    OutlineStrength=ScalarParameter() {defaultValue=1, group=Soft Outline, parameterName="Outline Strength (S)"}
    PixelDepthOffset=ScalarParameter() {defaultValue=1, group=Soft Outline, parameterName="PixelDepth Offset (S)"}
    MaxBlurDepth=ScalarParameter() {defaultValue=150, group=Soft Outline, parameterName="MaxBlurDepth (S)"}
    ------EnableInnerGlow------=ScalarParameter() {defaultValue=1, group=Soft Outline, parameterName=Enable Inner Glow}
    InnerGlowMaxArea=ScalarParameter() {defaultValue=100, group=Soft Outline, parameterName="InnerGlowMaxArea (S)"}
    InnerGlowPower=ScalarParameter() {defaultValue=10, group=Soft Outline, parameterName="InnerGlowPower (S)"}
    InnerGlowStrength=ScalarParameter() {defaultValue=0.5, group=Soft Outline, parameterName="InnerGlowStrength (S)"}
    ------LessUse------
    MeshPosition
    Distance Steps
    Radial Steps
   ) {materialFunction=/Game/Effects/Fort_Effects/Effects/Atlas/Materials/FXCommon_SoftOutline}.Result
   FadeDistance=ScalarParameter() {defaultValue=100, group=Soft Outline, parameterName=DepthFade}
  )
  n21 = Fresnel(ExponentIn=ScalarParameter() {defaultValue=1, group=Fresnel, parameterName=Fresnel_Exponent}) {baseReflectFraction=0, exponent=1}
  n22 = Power(
   Base=StaticSwitchParameter(True=OneMinus(None=@n21), False=@n21) {parameterName=Invert Fresnel}
   Exp=ScalarParameter() {group=Fresnel, parameterName=Fresnel_Power}
  )
  n23 = ScalarParameter() {defaultValue=2, parameterName=Vertical Mask Power}
  Multiply(
   A=Multiply(A=Multiply(A=@n19.A, B=@n20), B=Saturate(None=@n22))
   B=Multiply(A=Power(Base=@n17, Exp=@n23) {constExponent=1.5}, B=ScalarParameter() {defaultValue=2, parameterName=Vertical Mask Mult}) {constB=2}
  )
  Multiply(A=@n12, B=OneMinus(None=@n16.B))
 ]
}
