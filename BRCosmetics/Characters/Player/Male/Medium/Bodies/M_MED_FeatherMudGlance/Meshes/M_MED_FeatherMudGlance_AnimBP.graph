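# Animation blueprint for the FeatherMudGlance male player body cosmetic, adding secondary motion to hood, headphones, laces and shoelaces.
# The copied parent pose runs through bone offsets, constraints and Anim Dynamics into a cached pose.
# Three lace simulation variants branch from that cache, blend on skydiving and front-end flags, then feed a Control Rig.
# The event graph derives headphone offsets from arm, head and character state, and sets up dynamic materials.
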AnimBlueprint /BRCosmetics/Characters/Player/Male/Medium/Bodies/M_MED_FeatherMudGlance/Meshes/M_MED_FeatherMudGlance_AnimBP
target_skeleton=&/M_MED_FeatherMudGlance_Skeleton
graphs [
 AnimGraph {nodes [
   RigidBody() {
    bClampLinearTranslationLimitToRefPose=true
    bOverrideWorldGravity=true
    overridePhysicsAsset=&/M_MED_FeatherMudGlance_Hood_Physics
    alphaBoolBlend {blendOption=Linear}
    alphaScaleBiasClamp {bInterpResult=true}
    componentAppliedLinearAccClamp=<3000 3000 3000>
    componentLinearAccScale=<0.3 0.3 0.3>
    componentLinearVelScale=<0.3 0.3 0.3>
    overrideWorldGravity {z=-980}
   }
   n1 = "Constraint - mstr_lace_B (PP)"!(ComponentPose=Local To Component(LocalPose=Use cached pose 'OUTPUT'()), Alpha=1, ConstraintWeights_0=0.9, ConstraintWeights_1:float=0.1) -> (Pose) {
    alphaBoolBlend {blendOption=Linear}
    boneToModify {boneName=mstr_lace_B}
    constraintSetup [{targetBone {boneName=spine_05}}, {targetBone {boneName=neck_01}}]
    constraintWeights [1,1]
   }
   n2 = "Constraint - mstr_lace_A (PP)"!(ComponentPose=@n1, Alpha=1, ConstraintWeights_0=0.93, ConstraintWeights_1:float=0.07) -> (Pose) {
    alphaBoolBlend {blendOption=Linear}
    boneToModify {boneName=mstr_lace_A}
    constraintSetup [{targetBone {boneName=spine_05}}, {targetBone {boneName=neck_01}}]
    constraintWeights [1,1]
   }
   n3 = "Transform (Modify) Bone - Bone: mstr_lace_B"!(ComponentPose=@n2, bAlphaBoolEnabled:bool=True, Translation="0,0,0", Rotation="3,0,0", Scale="1,1,1") -> (Pose) {
    alphaInputType=Bool
    rotationMode=BMM_Additive
    rotationSpace=BCS_BoneSpace
    translationMode=BMM_Additive
    translationSpace=BCS_WorldSpace
    alphaBoolBlend {blendInTime=0.4, blendOption=Linear, blendOutTime=0.4}
    boneToModify {boneName=mstr_lace_B}
    rotation {pitch=3}
   }
   n4 = Anim Dynamics - Bone: dyn_lace_B_01!(ComponentPose=@n3, Alpha=0.1, GravityOverride:Vector="0,0,-980", ExternalForce="0,0,0") -> (Pose) {
    angularDampingOverride=0.999
    angularSpringConstant=500
    bAngularSpring=true
    bOverrideAngularDamping=true
    bUseGravityOverride=true
    bUseSphericalLimits=true
    gravityScale=-1
    simulationSpace=World
    alphaBoolBlend {blendOption=Linear}
    boundBone {boneName=dyn_lace_B_01}
    componentAppliedLinearAccClamp=<3000 3000 3000>
    componentLinearAccScale=<0.3 0.3 0.3>
    componentLinearVelScale=<0.3 0.3 0.3>
    gravityOverride {z=-980}
    physicsBodyDefinitions [{
      sphereCollisionRadius=4
      boundBone {boneName=dyn_lace_B_01}
      boxExtents=<16 10 10>
      constraintSetup {angularLimitsMax=<0 20 90>, angularLimitsMin=<0 -20 0>, angularTarget=<0 0 0>}
      localJointOffset=<3 0 0>
     }]
    sphericalLimits [{limitRadius=15, drivingBone {boneName=spine_05}, sphereLocalOffset=<-1.4488 -4.2514 0>}]
   }
   n5 = Anim Dynamics - Bone: dyn_lace_B_02!(ComponentPose=@n4, Alpha=0.1, GravityOverride:Vector="0,0,-980", ExternalForce="0,0,0") -> (Pose) {
    angularDampingOverride=0.999
    angularSpringConstant=500
    bAngularSpring=true
    bOverrideAngularDamping=true
    bUseGravityOverride=true
    bUseSphericalLimits=true
    gravityScale=-1
    simulationSpace=World
    alphaBoolBlend {blendOption=Linear}
    boundBone {boneName=dyn_lace_B_02}
    componentAppliedLinearAccClamp=<3000 3000 3000>
    componentLinearAccScale=<0.3 0.3 0.3>
    componentLinearVelScale=<0.3 0.3 0.3>
    gravityOverride {z=-980}
    physicsBodyDefinitions [{
      sphereCollisionRadius=2
      boundBone {boneName=dyn_lace_B_02}
      boxExtents=<16 10 10>
      constraintSetup {angularLimitsMax=<0 20 90>, angularLimitsMin=<0 -20 0>, angularTarget=<0 0 0>}
      localJointOffset=<3 0 0>
     }]
    sphericalLimits [{limitRadius=15, drivingBone {boneName=spine_05}, sphereLocalOffset=<-1.4488 -3.949 -2.8396>}]
   }
   n6 = Anim Dynamics - Bone: dyn_lace_B_03!(ComponentPose=@n5, Alpha=0.1, GravityOverride:Vector="0,0,-980", ExternalForce="0,0,0") -> (Pose) {
    angularDampingOverride=0.999
    angularSpringConstant=500
    bAngularSpring=true
    bOverrideAngularDamping=true
    bUseGravityOverride=true
    bUseSphericalLimits=true
    gravityScale=-1
    simulationSpace=World
    alphaBoolBlend {blendOption=Linear}
    boundBone {boneName=dyn_lace_B_03}
    componentAppliedLinearAccClamp=<3000 3000 3000>
    componentLinearAccScale=<0.3 0.3 0.3>
    componentLinearVelScale=<0.3 0.3 0.3>
    gravityOverride {z=-980}
    physicsBodyDefinitions [{
      sphereCollisionRadius=2
      boundBone {boneName=dyn_lace_B_03}
      boxExtents=<16 10 10>
      constraintSetup {angularLimitsMax=<0 20 90>, angularLimitsMin=<0 -20 0>, angularTarget=<0 0 0>}
      localJointOffset=<3 0 0>
     }]
    sphericalLimits [{limitRadius=15, drivingBone {boneName=spine_05}, sphereLocalOffset=<-3.4325 -2.4491 -3.4407>}]
   }
   n7 = "Transform (Modify) Bone - Bone: mstr_lace_A"!(ComponentPose=@n6, bAlphaBoolEnabled:bool=True, Translation="0,0,0", Rotation="3,0,0", Scale="1,1,1") -> (Pose) {
    alphaInputType=Bool
    rotationMode=BMM_Additive
    rotationSpace=BCS_BoneSpace
    translationMode=BMM_Additive
    translationSpace=BCS_WorldSpace
    alphaBoolBlend {blendInTime=0.4, blendOption=Linear, blendOutTime=0.4}
    boneToModify {boneName=mstr_lace_A}
    rotation {pitch=3}
   }
   n8 = Anim Dynamics - Bone: dyn_lace_A_01!(ComponentPose=@n7, Alpha=0.1, GravityOverride:Vector="0,0,-980", ExternalForce="0,0,0") -> (Pose) {
    angularDampingOverride=0.999
    angularSpringConstant=500
    bAngularSpring=true
    bOverrideAngularDamping=true
    bUseGravityOverride=true
    bUseSphericalLimits=true
    gravityScale=-1
    simulationSpace=World
    alphaBoolBlend {blendOption=Linear}
    boundBone {boneName=dyn_lace_A_01}
    componentAppliedLinearAccClamp=<3000 3000 3000>
    componentLinearAccScale=<0.3 0.3 0.3>
    componentLinearVelScale=<0.3 0.3 0.3>
    gravityOverride {z=-980}
    physicsBodyDefinitions [{
      sphereCollisionRadius=4
      boundBone {boneName=dyn_lace_A_01}
      boxExtents=<16 10 10>
      constraintSetup {angularLimitsMax=<0 20 90>, angularLimitsMin=<0 -20 0>, angularTarget=<0 0 0>}
      localJointOffset=<3 0 0>
     }]
    sphericalLimits [{limitRadius=15, drivingBone {boneName=spine_05}, sphereLocalOffset=<-1.4488 -4.2514 0>}]
   }
   n9 = Anim Dynamics - Bone: dyn_lace_A_02!(ComponentPose=@n8, Alpha=0.1, GravityOverride:Vector="0,0,-980", ExternalForce="0,0,0") -> (Pose) {
    angularDampingOverride=0.999
    angularSpringConstant=500
    bAngularSpring=true
    bOverrideAngularDamping=true
    bUseGravityOverride=true
    bUseSphericalLimits=true
    gravityScale=-1
    simulationSpace=World
    alphaBoolBlend {blendOption=Linear}
    boundBone {boneName=dyn_lace_A_02}
    componentAppliedLinearAccClamp=<3000 3000 3000>
    componentLinearAccScale=<0.3 0.3 0.3>
    componentLinearVelScale=<0.3 0.3 0.3>
    gravityOverride {z=-980}
    physicsBodyDefinitions [{
      sphereCollisionRadius=2
      boundBone {boneName=dyn_lace_A_02}
      boxExtents=<16 10 10>
      constraintSetup {angularLimitsMax=<0 20 90>, angularLimitsMin=<0 -20 0>, angularTarget=<0 0 0>}
      localJointOffset=<3 0 0>
     }]
    sphericalLimits [{limitRadius=15, drivingBone {boneName=spine_05}, sphereLocalOffset=<-1.4488 -4.2514 0>}]
   }
   n10 = Anim Dynamics - Bone: dyn_lace_A_02(ComponentPose=@n9, Alpha=0.1, GravityOverride="0,0,-980", ExternalForce="0,0,0") {
    angularDampingOverride=0.999
    angularSpringConstant=500
    bAngularSpring=true
    bOverrideAngularDamping=true
    bUseGravityOverride=true
    bUseSphericalLimits=true
    gravityScale=-1
    simulationSpace=World
    alphaBoolBlend {blendOption=Linear}
    boundBone {boneName=dyn_lace_A_02}
    componentAppliedLinearAccClamp=<3000 3000 3000>
    componentLinearAccScale=<0.3 0.3 0.3>
    componentLinearVelScale=<0.3 0.3 0.3>
    gravityOverride {z=-980}
    physicsBodyDefinitions [{
      sphereCollisionRadius=2
      boundBone {boneName=dyn_lace_A_02}
      boxExtents=<16 10 10>
      constraintSetup {angularLimitsMax=<0 20 90>, angularLimitsMin=<0 -20 0>, angularTarget=<0 0 0>}
      localJointOffset=<3 0 0>
     }]
    sphericalLimits [{limitRadius=15, drivingBone {boneName=spine_05}, sphereLocalOffset=<-1.4488 -3.317 0>}]
   }
   n11 = Component To Local(ComponentPose=@n10)
   n12 = "Constraint - mstr_lace_B (PP)"(ComponentPose=Local To Component(LocalPose=Use cached pose 'OUTPUT'()), Alpha=1, ConstraintWeights_0=0.9, ConstraintWeights_1=0.1) {
    alphaBoolBlend {blendOption=Linear}
    boneToModify {boneName=mstr_lace_B}
    constraintSetup [{targetBone {boneName=spine_05}}, {targetBone {boneName=neck_01}}]
    constraintWeights [1,1]
   }
   n13 = "Constraint - mstr_lace_A (PP)"(ComponentPose=@n12, Alpha=1, ConstraintWeights_0=0.93, ConstraintWeights_1=0.07) {
    alphaBoolBlend {blendOption=Linear}
    boneToModify {boneName=mstr_lace_A}
    constraintSetup [{targetBone {boneName=spine_05}}, {targetBone {boneName=neck_01}}]
    constraintWeights [1,1]
   }
   n14 = Get Gravity Override()
   n15 = Anim Dynamics - Bone: dyn_lace_B_01(ComponentPose=@n13, Alpha=1, GravityOverride=@n14, ExternalForce="0,0,0") {
    angularDampingOverride=0.999
    angularSpringConstant=500
    bAngularSpring=true
    bOverrideAngularDamping=true
    bUseGravityOverride=true
    bUseSphericalLimits=true
    gravityScale=-1
    simulationSpace=RootRelative
    alphaBoolBlend {blendOption=Linear}
    boundBone {boneName=dyn_lace_B_01}
    componentAppliedLinearAccClamp=<3000 3000 3000>
    componentLinearAccScale=<0.3 0.3 0.3>
    componentLinearVelScale=<0.3 0.3 0.3>
    gravityOverride {z=-980}
    physicsBodyDefinitions [{
      sphereCollisionRadius=4
      boundBone {boneName=dyn_lace_B_01}
      boxExtents=<16 10 10>
      constraintSetup {angularLimitsMax=<0 20 90>, angularLimitsMin=<0 -20 0>, angularTarget=<0 0 0>}
      localJointOffset=<3 0 0>
     }]
    sphericalLimits [{limitRadius=15, drivingBone {boneName=spine_05}, sphereLocalOffset=<-1.4488 -4.2514 0>}]
   }
   n16 = Anim Dynamics - Bone: dyn_lace_B_02(ComponentPose=@n15, Alpha=1, GravityOverride=@n14, ExternalForce="0,0,0") {
    angularDampingOverride=0.999
    angularSpringConstant=500
    bAngularSpring=true
    bOverrideAngularDamping=true
    bUseGravityOverride=true
    bUseSphericalLimits=true
    gravityScale=-1
    simulationSpace=RootRelative
    alphaBoolBlend {blendOption=Linear}
    boundBone {boneName=dyn_lace_B_02}
    componentAppliedLinearAccClamp=<3000 3000 3000>
    componentLinearAccScale=<0.3 0.3 0.3>
    componentLinearVelScale=<0.3 0.3 0.3>
    gravityOverride {z=-980}
    physicsBodyDefinitions [{
      sphereCollisionRadius=2
      boundBone {boneName=dyn_lace_B_02}
      boxExtents=<16 10 10>
      constraintSetup {angularLimitsMax=<0 20 90>, angularLimitsMin=<0 -20 0>, angularTarget=<0 0 0>}
      localJointOffset=<3 0 0>
     }]
    sphericalLimits [{limitRadius=15, drivingBone {boneName=spine_05}, sphereLocalOffset=<-1.4488 -3.949 -2.8396>}]
   }
   n17 = Anim Dynamics - Bone: dyn_lace_B_03(ComponentPose=@n16, Alpha=1, GravityOverride=@n14, ExternalForce="0,0,0") {
    angularDampingOverride=0.999
    angularSpringConstant=500
    bAngularSpring=true
    bOverrideAngularDamping=true
    bUseGravityOverride=true
    bUseSphericalLimits=true
    gravityScale=-1
    simulationSpace=RootRelative
    alphaBoolBlend {blendOption=Linear}
    boundBone {boneName=dyn_lace_B_03}
    componentAppliedLinearAccClamp=<3000 3000 3000>
    componentLinearAccScale=<0.3 0.3 0.3>
    componentLinearVelScale=<0.3 0.3 0.3>
    gravityOverride {z=-980}
    physicsBodyDefinitions [{
      sphereCollisionRadius=2
      boundBone {boneName=dyn_lace_B_03}
      boxExtents=<16 10 10>
      constraintSetup {angularLimitsMax=<0 20 90>, angularLimitsMin=<0 -20 0>, angularTarget=<0 0 0>}
      localJointOffset=<3 0 0>
     }]
    sphericalLimits [{limitRadius=15, drivingBone {boneName=spine_05}, sphereLocalOffset=<-3.4325 -2.4491 -3.4407>}]
   }
   n18 = Anim Dynamics - Bone: dyn_lace_A_01(ComponentPose=@n17, Alpha=1, GravityOverride=@n14, ExternalForce="0,0,0") {
    angularDampingOverride=0.999
    angularSpringConstant=500
    bAngularSpring=true
    bOverrideAngularDamping=true
    bUseGravityOverride=true
    bUseSphericalLimits=true
    gravityScale=-1
    simulationSpace=RootRelative
    alphaBoolBlend {blendOption=Linear}
    boundBone {boneName=dyn_lace_A_01}
    componentAppliedLinearAccClamp=<3000 3000 3000>
    componentLinearAccScale=<0.3 0.3 0.3>
    componentLinearVelScale=<0.3 0.3 0.3>
    gravityOverride {z=-980}
    physicsBodyDefinitions [{
      sphereCollisionRadius=4
      boundBone {boneName=dyn_lace_A_01}
      boxExtents=<16 10 10>
      constraintSetup {angularLimitsMax=<0 20 90>, angularLimitsMin=<0 -20 0>, angularTarget=<0 0 0>}
      localJointOffset=<3 0 0>
     }]
    sphericalLimits [{limitRadius=15, drivingBone {boneName=spine_05}, sphereLocalOffset=<-1.4488 -4.2514 0>}]
   }
   n19 = Anim Dynamics - Bone: dyn_lace_A_02(ComponentPose=@n18, Alpha=1, GravityOverride=@n14, ExternalForce="0,0,0") {
    angularDampingOverride=0.999
    angularSpringConstant=500
    bAngularSpring=true
    bOverrideAngularDamping=true
    bUseGravityOverride=true
    bUseSphericalLimits=true
    gravityScale=-1
    simulationSpace=RootRelative
    alphaBoolBlend {blendOption=Linear}
    boundBone {boneName=dyn_lace_A_02}
    componentAppliedLinearAccClamp=<3000 3000 3000>
    componentLinearAccScale=<0.3 0.3 0.3>
    componentLinearVelScale=<0.3 0.3 0.3>
    gravityOverride {z=-980}
    physicsBodyDefinitions [{
      sphereCollisionRadius=2
      boundBone {boneName=dyn_lace_A_02}
      boxExtents=<16 10 10>
      constraintSetup {angularLimitsMax=<0 20 90>, angularLimitsMin=<0 -20 0>, angularTarget=<0 0 0>}
      localJointOffset=<3 0 0>
     }]
    sphericalLimits [{limitRadius=15, drivingBone {boneName=spine_05}, sphereLocalOffset=<-1.4488 -4.2514 0>}]
   }
   n20 = Anim Dynamics - Bone: dyn_lace_A_02(ComponentPose=@n19, Alpha=1, GravityOverride=@n14, ExternalForce="0,0,0") {
    angularDampingOverride=0.999
    angularSpringConstant=500
    bAngularSpring=true
    bOverrideAngularDamping=true
    bUseGravityOverride=true
    bUseSphericalLimits=true
    gravityScale=-1
    simulationSpace=RootRelative
    alphaBoolBlend {blendOption=Linear}
    boundBone {boneName=dyn_lace_A_02}
    componentAppliedLinearAccClamp=<3000 3000 3000>
    componentLinearAccScale=<0.3 0.3 0.3>
    componentLinearVelScale=<0.3 0.3 0.3>
    gravityOverride {z=-980}
    physicsBodyDefinitions [{
      sphereCollisionRadius=2
      boundBone {boneName=dyn_lace_A_02}
      boxExtents=<16 10 10>
      constraintSetup {angularLimitsMax=<0 20 90>, angularLimitsMin=<0 -20 0>, angularTarget=<0 0 0>}
      localJointOffset=<3 0 0>
     }]
    sphericalLimits [{limitRadius=15, drivingBone {boneName=spine_05}, sphereLocalOffset=<-1.4488 -3.317 0>}]
   }
   n21 = Component To Local(ComponentPose=@n20)
   n22 = Get bIsInFrontEnd()
   n23 = "Constraint - mstr_lace_B (PP)"(ComponentPose=Local To Component(LocalPose=Use cached pose 'OUTPUT'()), Alpha=1, ConstraintWeights_0=0.9, ConstraintWeights_1=0.1) {
    alphaBoolBlend {blendOption=Linear}
    boneToModify {boneName=mstr_lace_B}
    constraintSetup [{targetBone {boneName=spine_05}}, {targetBone {boneName=neck_01}}]
    constraintWeights [1,1]
   }
   n24 = "Constraint - mstr_lace_A (PP)"(ComponentPose=@n23, Alpha=1, ConstraintWeights_0=0.93, ConstraintWeights_1=0.07) {
    alphaBoolBlend {blendOption=Linear}
    boneToModify {boneName=mstr_lace_A}
    constraintSetup [{targetBone {boneName=spine_05}}, {targetBone {boneName=neck_01}}]
    constraintWeights [1,1]
   }
   n25 = Anim Dynamics - Bone: dyn_lace_B_01(ComponentPose=@n24, Alpha=1, GravityOverride="0,0,-980", ExternalForce="0,0,0") {
    angularDampingOverride=0.999
    angularSpringConstant=500
    bAngularSpring=true
    bOverrideAngularDamping=true
    bUseGravityOverride=true
    bUseSphericalLimits=true
    gravityScale=-1
    simulationSpace=RootRelative
    alphaBoolBlend {blendOption=Linear}
    boundBone {boneName=dyn_lace_B_01}
    componentAppliedLinearAccClamp=<3000 3000 3000>
    componentLinearAccScale=<0.3 0.3 0.3>
    componentLinearVelScale=<0.3 0.3 0.3>
    gravityOverride {z=-980}
    physicsBodyDefinitions [{
      sphereCollisionRadius=4
      boundBone {boneName=dyn_lace_B_01}
      boxExtents=<16 10 10>
      constraintSetup {angularLimitsMax=<0 20 90>, angularLimitsMin=<0 -20 0>, angularTarget=<0 0 0>}
      localJointOffset=<3 0 0>
     }]
    sphericalLimits [{limitRadius=15, drivingBone {boneName=spine_05}, sphereLocalOffset=<-1.4488 -4.2514 0>}]
   }
   n26 = Anim Dynamics - Bone: dyn_lace_B_02(ComponentPose=@n25, Alpha=1, GravityOverride="0,0,-980", ExternalForce="0,0,0") {
    angularDampingOverride=0.999
    angularSpringConstant=500
    bAngularSpring=true
    bOverrideAngularDamping=true
    bUseGravityOverride=true
    bUseSphericalLimits=true
    gravityScale=-1
    simulationSpace=RootRelative
    alphaBoolBlend {blendOption=Linear}
    boundBone {boneName=dyn_lace_B_02}
    componentAppliedLinearAccClamp=<3000 3000 3000>
    componentLinearAccScale=<0.3 0.3 0.3>
    componentLinearVelScale=<0.3 0.3 0.3>
    gravityOverride {z=-980}
    physicsBodyDefinitions [{
      sphereCollisionRadius=2
      boundBone {boneName=dyn_lace_B_02}
      boxExtents=<16 10 10>
      constraintSetup {angularLimitsMax=<0 20 90>, angularLimitsMin=<0 -20 0>, angularTarget=<0 0 0>}
      localJointOffset=<3 0 0>
     }]
    sphericalLimits [{limitRadius=15, drivingBone {boneName=spine_05}, sphereLocalOffset=<-1.4488 -3.949 -2.8396>}]
   }
   n27 = Anim Dynamics - Bone: dyn_lace_B_03(ComponentPose=@n26, Alpha=1, GravityOverride="0,0,-980", ExternalForce="0,0,0") {
    angularDampingOverride=0.999
    angularSpringConstant=500
    bAngularSpring=true
    bOverrideAngularDamping=true
    bUseGravityOverride=true
    bUseSphericalLimits=true
    gravityScale=-1
    simulationSpace=RootRelative
    alphaBoolBlend {blendOption=Linear}
    boundBone {boneName=dyn_lace_B_03}
    componentAppliedLinearAccClamp=<3000 3000 3000>
    componentLinearAccScale=<0.3 0.3 0.3>
    componentLinearVelScale=<0.3 0.3 0.3>
    gravityOverride {z=-980}
    physicsBodyDefinitions [{
      sphereCollisionRadius=2
      boundBone {boneName=dyn_lace_B_03}
      boxExtents=<16 10 10>
      constraintSetup {angularLimitsMax=<0 20 90>, angularLimitsMin=<0 -20 0>, angularTarget=<0 0 0>}
      localJointOffset=<3 0 0>
     }]
    sphericalLimits [{limitRadius=15, drivingBone {boneName=spine_05}, sphereLocalOffset=<-3.4325 -2.4491 -3.4407>}]
   }
   n28 = Anim Dynamics - Bone: dyn_lace_A_01(ComponentPose=@n27, Alpha=1, GravityOverride="0,0,-980", ExternalForce="0,0,0") {
    angularDampingOverride=0.999
    angularSpringConstant=500
    bAngularSpring=true
    bOverrideAngularDamping=true
    bUseGravityOverride=true
    bUseSphericalLimits=true
    gravityScale=-1
    simulationSpace=RootRelative
    alphaBoolBlend {blendOption=Linear}
    boundBone {boneName=dyn_lace_A_01}
    componentAppliedLinearAccClamp=<3000 3000 3000>
    componentLinearAccScale=<0.3 0.3 0.3>
    componentLinearVelScale=<0.3 0.3 0.3>
    gravityOverride {z=-980}
    physicsBodyDefinitions [{
      sphereCollisionRadius=4
      boundBone {boneName=dyn_lace_A_01}
      boxExtents=<16 10 10>
      constraintSetup {angularLimitsMax=<0 20 90>, angularLimitsMin=<0 -20 0>, angularTarget=<0 0 0>}
      localJointOffset=<3 0 0>
     }]
    sphericalLimits [{limitRadius=15, drivingBone {boneName=spine_05}, sphereLocalOffset=<-1.4488 -4.2514 0>}]
   }
   n29 = Anim Dynamics - Bone: dyn_lace_A_02(ComponentPose=@n28, Alpha=1, GravityOverride="0,0,-980", ExternalForce="0,0,0") {
    angularDampingOverride=0.999
    angularSpringConstant=500
    bAngularSpring=true
    bOverrideAngularDamping=true
    bUseGravityOverride=true
    bUseSphericalLimits=true
    gravityScale=-1
    simulationSpace=RootRelative
    alphaBoolBlend {blendOption=Linear}
    boundBone {boneName=dyn_lace_A_02}
    componentAppliedLinearAccClamp=<3000 3000 3000>
    componentLinearAccScale=<0.3 0.3 0.3>
    componentLinearVelScale=<0.3 0.3 0.3>
    gravityOverride {z=-980}
    physicsBodyDefinitions [{
      sphereCollisionRadius=2
      boundBone {boneName=dyn_lace_A_02}
      boxExtents=<16 10 10>
      constraintSetup {angularLimitsMax=<0 20 90>, angularLimitsMin=<0 -20 0>, angularTarget=<0 0 0>}
      localJointOffset=<3 0 0>
     }]
    sphericalLimits [{limitRadius=15, drivingBone {boneName=spine_05}, sphereLocalOffset=<-1.4488 -4.2514 0>}]
   }
   n30 = Anim Dynamics - Bone: dyn_lace_A_02(ComponentPose=@n29, Alpha=1, GravityOverride="0,0,-980", ExternalForce="0,0,0") {
    angularDampingOverride=0.999
    angularSpringConstant=500
    bAngularSpring=true
    bOverrideAngularDamping=true
    bUseGravityOverride=true
    bUseSphericalLimits=true
    gravityScale=-1
    simulationSpace=RootRelative
    alphaBoolBlend {blendOption=Linear}
    boundBone {boneName=dyn_lace_A_02}
    componentAppliedLinearAccClamp=<3000 3000 3000>
    componentLinearAccScale=<0.3 0.3 0.3>
    componentLinearVelScale=<0.3 0.3 0.3>
    gravityOverride {z=-980}
    physicsBodyDefinitions [{
      sphereCollisionRadius=2
      boundBone {boneName=dyn_lace_A_02}
      boxExtents=<16 10 10>
      constraintSetup {angularLimitsMax=<0 20 90>, angularLimitsMin=<0 -20 0>, angularTarget=<0 0 0>}
      localJointOffset=<3 0 0>
     }]
    sphericalLimits [{limitRadius=15, drivingBone {boneName=spine_05}, sphereLocalOffset=<-1.4488 -3.317 0>}]
   }
   n31 = Control Rig(Source=Blend Poses by bool(
     BlendPose_0=Component To Local(ComponentPose=@n30)
     BlendPose_1=Blend Poses by bool(BlendPose_0=@n11, BlendPose_1=@n21, BlendTime_0=0.3, BlendTime_1=0.3, bActiveValue=@n22) {blendType=Linear}
     BlendTime_0=0.3
     BlendTime_1=0.3
     bActiveValue=Get bIsSkyDiving()
    ) {blendType=Linear}) {
    alphaBoolBlend {blendOption=Linear}
    controlRigAssetReference {blueprintRigClass=&/M_MED_FeatherMudGlance_CtrlRig.M_MED_FeatherMudGlance_CtrlRig_C}
   }
   Output Pose(Result=@n31) {layerGroup=None}
   n32 = Get is in jump or falling() -> (is in jump or falling:bool)
   n33 = Get jump easing tr() -> (jump easing tr:double)
   n34 = Get Gravity Override()
   n35 = Get Gravity Override()
   n36 = "Transform (Modify) Bone - Bone: dyn_headphones_MSTR"!(
    ComponentPose=Local To Component(LocalPose=Copy Pose From Mesh() {bUseAttachedParent=true})
    bAlphaBoolEnabled:bool=True
    Translation="0,0,0"
    Rotation="0,0,0"
    Scale="1,1,1"
   ) -> (Pose) {
    alphaInputType=Bool
    rotationMode=BMM_Additive
    rotationSpace=BCS_BoneSpace
    translationMode=BMM_Additive
    translationSpace=BCS_BoneSpace
    alphaBoolBlend {blendInTime=0.2, blendOption=Linear, blendOutTime=0.4}
    boneToModify {boneName=dyn_headphones_MSTR}
   }
   n37 = "Constraint - dyn_headphones (RR)"!(ComponentPose=@n36, Alpha=1, ConstraintWeights_0=0.3, ConstraintWeights_1:float=0.7) -> (Pose) {
    alphaBoolBlend {blendOption=Linear}
    boneToModify {boneName=dyn_headphones}
    constraintSetup [{transformType=Rotation, targetBone {boneName=head}}, {transformType=Rotation, targetBone {boneName=spine_05}}]
    constraintWeights [1,1]
   }
   n38 = Look At - Bone: dyn_headphones_MSTR(ComponentPose=@n37, Alpha=1) {
    alphaBoolBlend {blendOption=Linear}
    boneToModify {boneName=dyn_headphones_MSTR}
    lookAtLocation=<6.8482 36.3494>
    lookAtTarget {boneReference {boneName=neck_01}}
   }
   n39 = "Transform (Modify) Bone - Bone: dyn_headphones"!(
    ComponentPose=@n38
    bAlphaBoolEnabled:bool=Get Heaphone_crouch() -> (Heaphone_crouch:bool)
    Translation="-0.2,0,0"
    Rotation="0,0,0"
    Scale="1,1,1"
   ) -> (Pose) {
    alphaInputType=Bool
    rotationMode=BMM_Additive
    rotationSpace=BCS_BoneSpace
    translationMode=BMM_Additive
    translationSpace=BCS_BoneSpace
    alphaBoolBlend {blendInTime=0.2, blendOption=Linear, blendOutTime=0.4}
    boneToModify {boneName=dyn_headphones}
    translation {x=-0.2}
   }
   n40 = "Transform (Modify) Bone - Bone: dyn_headphones"(ComponentPose=@n39, bAlphaBoolEnabled=Get bIsTargeting(), Translation="-0.6,0,0", Rotation="0,0,0", Scale="1,1,1") {
    alphaInputType=Bool
    rotationMode=BMM_Additive
    rotationSpace=BCS_BoneSpace
    translationMode=BMM_Additive
    translationSpace=BCS_BoneSpace
    alphaBoolBlend {blendInTime=0.2, blendOption=Linear, blendOutTime=0.4}
    boneToModify {boneName=dyn_headphones}
    translation {x=-0.6}
   }
   n41 = "Transform (Modify) Bone - Bone: dyn_headphones_r_MSTR"(ComponentPose=@n40, Alpha=Get Headphone_Right() -> (Headphone_Right:float), Translation="0,0,0", Rotation="0,0,15", Scale="1,1,1") {
    rotationMode=BMM_Additive
    rotationSpace=BCS_BoneSpace
    translationMode=BMM_Additive
    translationSpace=BCS_BoneSpace
    alphaBoolBlend {blendInTime=0.2, blendOption=Linear, blendOutTime=0.6}
    boneToModify {boneName=dyn_headphones_r_MSTR}
    rotation {roll=15}
   }
   n42 = "Transform (Modify) Bone - Bone: dyn_headphones_l_MSTR"(ComponentPose=@n41, Alpha=Get Left_ear() -> (Left_ear:double), Translation="0,3,0", Rotation="0,-5,8", Scale="1,1,1") {
    rotationMode=BMM_Additive
    rotationSpace=BCS_BoneSpace
    translationMode=BMM_Additive
    translationSpace=BCS_BoneSpace
    alphaBoolBlend {blendInTime=0.2, blendOption=Linear, blendOutTime=0.6}
    boneToModify {boneName=dyn_headphones_l_MSTR}
    rotation {roll=8, yaw=-5}
    translation {y=3}
   }
   n43 = Get bIsPlayingMeleeAnim()
   n44 = "Constraint - dyn_headphones (RR)"!(ComponentPose=@n42, bAlphaBoolEnabled:bool=@n43, ConstraintWeights_0=0.3, ConstraintWeights_1=0.7) -> (Pose) {
    alphaInputType=Bool
    alphaBoolBlend {blendInTime=0.2, blendOption=Linear, blendOutTime=0.8}
    boneToModify {boneName=dyn_headphones}
    constraintSetup [{transformType=Rotation, targetBone {boneName=spine_05}}, {transformType=Rotation, targetBone {boneName=head}}]
    constraintWeights [1,1]
   }
   n45 = "Transform (Modify) Bone - Bone: dyn_headphones"(ComponentPose=@n44, bAlphaBoolEnabled=@n43, Translation="-1,0,0", Rotation="0,0,0", Scale="1,1,1") {
    alphaInputType=Bool
    rotationMode=BMM_Additive
    rotationSpace=BCS_BoneSpace
    translationMode=BMM_Additive
    translationSpace=BCS_BoneSpace
    alphaBoolBlend {blendInTime=0.2, blendOption=Linear, blendOutTime=0.8}
    boneToModify {boneName=dyn_headphones}
    translation {x=-1}
   }
   n46 = Anim Dynamics - Bone: dyn_shoelace_left_side_A!(ComponentPose=@n45, Alpha=1, GravityOverride:Vector=@n35, ExternalForce="0,0,0") -> (Pose) {
    angularDampingOverride=0.999
    angularSpringConstant=500
    bAngularSpring=true
    bOverrideAngularDamping=true
    bUseGravityOverride=true
    bUseSphericalLimits=true
    gravityScale=-1
    simulationSpace=RootRelative
    alphaBoolBlend {blendOption=Linear}
    boundBone {boneName=dyn_shoelace_left_side_A}
    componentAppliedLinearAccClamp=<3000 3000 3000>
    componentLinearAccScale=<0.3 0.3 0.3>
    componentLinearVelScale=<0.3 0.3 0.3>
    gravityOverride {z=-980}
    physicsBodyDefinitions [{
      sphereCollisionRadius=4
      boundBone {boneName=dyn_shoelace_left_side_A}
      boxExtents=<16 10 10>
      constraintSetup {angularLimitsMax=<0 0 5>, angularLimitsMin=<0 -45 -5>, angularTarget=<0 0 0>}
      localJointOffset=<3 0 0>
     }]
    sphericalLimits [{limitRadius=5.6246, drivingBone {boneName=foot_l}, sphereLocalOffset=<-4.5693 6.6143 -0.8299>}]
   }
   n47 = Anim Dynamics - Bone: dyn_shoelace_left_side_B!(ComponentPose=@n46, Alpha=1, GravityOverride:Vector=@n35, ExternalForce="0,0,0") -> (Pose) {
    angularDampingOverride=0.999
    angularSpringConstant=500
    bAngularSpring=true
    bOverrideAngularDamping=true
    bUseGravityOverride=true
    bUseSphericalLimits=true
    gravityScale=-1
    simulationSpace=RootRelative
    alphaBoolBlend {blendOption=Linear}
    boundBone {boneName=dyn_shoelace_left_side_B}
    componentAppliedLinearAccClamp=<3000 3000 3000>
    componentLinearAccScale=<0.3 0.3 0.3>
    componentLinearVelScale=<0.3 0.3 0.3>
    gravityOverride {z=-980}
    physicsBodyDefinitions [{
      sphereCollisionRadius=4
      boundBone {boneName=dyn_shoelace_left_side_B}
      boxExtents=<16 10 10>
      constraintSetup {angularLimitsMax=<0 0 5>, angularLimitsMin=<0 -45 -5>, angularTarget=<0 0 0>}
      localJointOffset=<3 0 0>
     }]
    sphericalLimits [{limitRadius=2, drivingBone {boneName=foot_l}, sphereLocalOffset=<0 0 0>}]
   }
   n48 = Anim Dynamics - Bone: dyn_shoelace_right_side_B!(ComponentPose=@n47, Alpha=1, GravityOverride:Vector=@n34, ExternalForce="0,0,0") -> (Pose) {
    angularDampingOverride=0.999
    angularSpringConstant=500
    bAngularSpring=true
    bOverrideAngularDamping=true
    bUseGravityOverride=true
    bUseSphericalLimits=true
    gravityScale=-1
    simulationSpace=RootRelative
    alphaBoolBlend {blendOption=Linear}
    boundBone {boneName=dyn_shoelace_right_side_B}
    componentAppliedLinearAccClamp=<3000 3000 3000>
    componentLinearAccScale=<0.3 0.3 0.3>
    componentLinearVelScale=<0.3 0.3 0.3>
    gravityOverride {z=-980}
    physicsBodyDefinitions [{
      sphereCollisionRadius=4
      boundBone {boneName=dyn_shoelace_right_side_B}
      boxExtents=<16 10 10>
      constraintSetup {angularLimitsMax=<0 0 5>, angularLimitsMin=<0 -45 -5>, angularTarget=<0 0 0>}
      localJointOffset=<3 0 0>
     }]
    sphericalLimits [{limitRadius=5.6246, drivingBone {boneName=foot_r}, sphereLocalOffset=<-4.5693 6.6143 -0.8299>}]
   }
   n49 = Anim Dynamics - Bone: dyn_shoelace_right_side_A!(ComponentPose=@n48, Alpha=1, GravityOverride:Vector=@n34, ExternalForce="0,0,0") -> (Pose) {
    angularDampingOverride=0.999
    angularSpringConstant=500
    bAngularSpring=true
    bOverrideAngularDamping=true
    bUseGravityOverride=true
    bUseSphericalLimits=true
    gravityScale=-1
    simulationSpace=RootRelative
    alphaBoolBlend {blendOption=Linear}
    boundBone {boneName=dyn_shoelace_right_side_A}
    componentAppliedLinearAccClamp=<3000 3000 3000>
    componentLinearAccScale=<0.3 0.3 0.3>
    componentLinearVelScale=<0.3 0.3 0.3>
    gravityOverride {z=-980}
    physicsBodyDefinitions [{
      sphereCollisionRadius=4
      boundBone {boneName=dyn_shoelace_right_side_A}
      boxExtents=<16 10 10>
      constraintSetup {angularLimitsMax=<0 0 5>, angularLimitsMin=<0 -45 -5>, angularTarget=<0 0 0>}
      localJointOffset=<3 0 0>
     }]
    sphericalLimits [{limitRadius=2, drivingBone {boneName=foot_r}, sphereLocalOffset=<0 0 0>}]
   }
   n50 = "Transform (Modify) Bone - Bone: dyn_hood_01"!(
    ComponentPose=@n49
    bAlphaBoolEnabled:bool=@n32
    Translation_X:double=0
    Translation_Y:double=0
    Translation_Z:double=@n33
    Rotation_Roll:float=0
    Rotation_Pitch:float=0
    Rotation_Yaw:float=Get jump easing() -> (jump easing:double)
    Scale="1,1,1"
   ) -> (Pose) {
    alphaInputType=Bool
    rotationMode=BMM_Additive
    rotationSpace=BCS_BoneSpace
    translationMode=BMM_Additive
    translationSpace=BCS_WorldSpace
    alphaBoolBlend {blendInTime=0.4, blendOption=Linear, blendOutTime=0.4}
    boneToModify {boneName=dyn_hood_01}
   }
   n51 = Get side jump easing() -> (side jump easing:double)
   n52 = "Transform (Modify) Bone - Bone: dyn_hood_l_01"!(
    ComponentPose=@n50
    bAlphaBoolEnabled:bool=@n32
    Translation_X:double=0
    Translation_Y:double=0
    Translation_Z:double=@n33
    Rotation_Roll:float=0
    Rotation_Pitch:float=0
    Rotation_Yaw:float=@n51
    Scale="1,1,1"
   ) -> (Pose) {
    alphaInputType=Bool
    rotationMode=BMM_Additive
    rotationSpace=BCS_BoneSpace
    translationMode=BMM_Additive
    translationSpace=BCS_WorldSpace
    alphaBoolBlend {blendInTime=0.4, blendOption=Linear, blendOutTime=0.4}
    boneToModify {boneName=dyn_hood_l_01}
   }
   n53 = "Transform (Modify) Bone - Bone: dyn_hood_r_01"!(
    ComponentPose=@n52
    bAlphaBoolEnabled:bool=@n32
    Translation_X:double=0
    Translation_Y:double=0
    Translation_Z:double=@n33
    Rotation_Roll:float=0
    Rotation_Pitch:float=0
    Rotation_Yaw:float=@n51
    Scale="1,1,1"
   ) -> (Pose) {
    alphaInputType=Bool
    rotationMode=BMM_Additive
    rotationSpace=BCS_BoneSpace
    translationSpace=BCS_WorldSpace
    alphaBoolBlend {blendInTime=0.4, blendOption=Linear, blendOutTime=0.4}
    boneToModify {boneName=dyn_hood_r_01}
   }
   n54 = "Transform (Modify) Bone - Bone: dyn_hood_01"!(ComponentPose=@n53, Alpha:float=1, Translation:Vector="-5,-2,0", Rotation:Rotator="0,0,0", Scale="1,1,1") -> (Pose) {
    rotationMode=BMM_Additive
    rotationSpace=BCS_BoneSpace
    translationMode=BMM_Additive
    translationSpace=BCS_BoneSpace
    alphaBoolBlend {blendOption=Linear}
    boneToModify {boneName=dyn_hood_01}
    translation=<-5 -2>
   }
   n55 = "Transform (Modify) Bone - Bone: dyn_hood_l_01"!(ComponentPose=@n54, Alpha:float=1, Translation:Vector="3,1,0", Rotation:Rotator="0,0,0", Scale="1,1,1") -> (Pose) {
    rotationMode=BMM_Additive
    rotationSpace=BCS_BoneSpace
    translationMode=BMM_Additive
    translationSpace=BCS_BoneSpace
    alphaBoolBlend {blendOption=Linear}
    boneToModify {boneName=dyn_hood_l_01}
    translation=<3 1>
   }
   n56 = "Transform (Modify) Bone - Bone: dyn_hood_r_01"!(ComponentPose=@n55, Alpha:float=1, Translation:Vector="-3,-1,0", Rotation:Rotator="0,0,0", Scale="1,1,1") -> (Pose) {
    rotationMode=BMM_Additive
    rotationSpace=BCS_BoneSpace
    translationMode=BMM_Additive
    translationSpace=BCS_BoneSpace
    alphaBoolBlend {blendOption=Linear}
    boneToModify {boneName=dyn_hood_r_01}
    translation=<-3 -1>
   }
   n57 = "Transform (Modify) Bone - Bone: dyn_hoodieTop"!(ComponentPose=@n56, bAlphaBoolEnabled:bool=True, Translation="-2.5,-1.468,0.353", Rotation="0,0,0", Scale="1,1,1") -> (Pose) {
    alphaInputType=Bool
    rotationMode=BMM_Additive
    rotationSpace=BCS_BoneSpace
    scaleMode=BMM_Additive
    scaleSpace=BCS_BoneSpace
    translationMode=BMM_Additive
    translationSpace=BCS_BoneSpace
    alphaBoolBlend {blendInTime=0.1, blendOption=Linear, blendOutTime=0.1}
    boneToModify {boneName=dyn_hoodieTop}
    translation=<-2.5 -1.468 0.353>
   }
   n58 = "Transform (Modify) Bone - Bone: dyn_hoodieTop"(ComponentPose=@n57, bAlphaBoolEnabled=True, Translation="4,1.468,-0.353", Rotation="0,0,0", Scale="1,1,1") {
    alphaInputType=Bool
    rotationMode=BMM_Additive
    rotationSpace=BCS_BoneSpace
    scaleMode=BMM_Additive
    scaleSpace=BCS_BoneSpace
    translationMode=BMM_Additive
    translationSpace=BCS_BoneSpace
    alphaBoolBlend {blendInTime=0.1, blendOption=Linear, blendOutTime=0.1}
    boneToModify {boneName=dyn_hoodieTop}
    translation=<4 1.468 -0.353>
   }
   n59 = Anim Dynamics - Bone: dyn_hood_01!(ComponentPose=@n58, Alpha=0.85, GravityOverride:Vector="0,0,0", ExternalForce="0,0,0") -> (Pose) {
    angularDampingOverride=0.9999
    angularSpringConstant=75
    bAngularSpring=true
    bGravityOverrideInSimSpace=true
    bOverrideAngularDamping=true
    simulationSpace=RootRelative
    alphaBoolBlend {blendOption=Linear}
    boundBone {boneName=dyn_hood_01}
    componentAppliedLinearAccClamp=<5000 5000 5000>
    componentLinearAccScale=<1 1 1>
    componentLinearVelScale=<1 1 1>
    physicsBodyDefinitions [{
      sphereCollisionRadius=0
      boundBone {boneName=dyn_hood_01}
      boxExtents=<10 20 20>
      constraintSetup {angularLimitsMax=<0 0 0>, angularLimitsMin=<0 -70 0>, angularTarget=<0 0 0>}
      localJointOffset=<2 0 0>
     }]
   }
   n60 = Anim Dynamics - Bone: dyn_hood_02!(ComponentPose=@n59, Alpha=0.85, GravityOverride:Vector="0,0,0", ExternalForce="0,0,0") -> (Pose) {
    angularDampingOverride=0.9999
    angularSpringConstant=75
    bAngularSpring=true
    bGravityOverrideInSimSpace=true
    bOverrideAngularDamping=true
    simulationSpace=RootRelative
    alphaBoolBlend {blendOption=Linear}
    boundBone {boneName=dyn_hood_02}
    componentAppliedLinearAccClamp=<5000 5000 5000>
    componentLinearAccScale=<1 1 1>
    componentLinearVelScale=<1 1 1>
    physicsBodyDefinitions [{
      sphereCollisionRadius=0
      boundBone {boneName=dyn_hood_02}
      boxExtents=<10 20 20>
      constraintSetup {angularLimitsMax=<0 0 0>, angularLimitsMin=<0 -50 0>, angularTarget=<0 0 0>}
      localJointOffset=<2 0 0>
     }]
   }
   Save cached pose 'OUTPUT'(Pose=Component To Local(ComponentPose=@n60)) {cacheName=OUTPUT}
   Get bIsSkydiveDiveMode()
  ]}
 EventGraph {nodes [
   Event OnCharacterCustomizationLoaded()
   n1 = GetPartSkeletalMeshComponent()
   > n2 = Get Dynamic Material Instance(SkeletalMeshComponent=@n1)
   > Set Mat Instance Dynamic(Mat Instance Dynamic=@n2)
   > n3 = GetOrCreateDynamicOverlayMaterial(MeshComponent=@n1, MaterialName=None)
   > Set Overlay Mat Dynamic(Overlay Mat Dynamic:Material Instance Dynamic Object Reference=@n3) -> (Output_Get:Material Instance Dynamic Object Reference)
   n4 = Event BlueprintUpdateAnimation()
   > n5 = Sequence()
   n6 = NOT Boolean(A=OR Boolean(A=Get bIsJumping(), B=Get bIsFalling()))
   n7 = Get bIsMoving2D()
   n8 = float * float(A=Vector Length(A=Get PawnVelocity()), B="To Float (Boolean)"(InBool=AND Boolean(A=@n7, B=@n6))) {is=title K2Node_PromotableOperator}
   n9 = Get bIsPlayingMeleeAnim()
   > Set rigid_body_alpha(rigid_body_alpha:double=float * float!(A=SelectFloat(A=0.2, B=0.7, bPickA=@n9), B=MapRangeClamped(Value=@n8, InRangeA=550, InRangeB=770, OutRangeA=0.7, OutRangeB=0.3), C:double) -> (ReturnValue)) -> (Output_Get:double)
   > Set Constraint alpha(Constraint alpha:double=MapRangeClamped(Value=@n8, InRangeA=30, InRangeB=550, OutRangeB=0.3)) -> (Output_Get:double)
   n10 = GetAlphaFromReferenceTransformDelta(
    SkeletalMeshComponent=GetOwnerSkeletalMeshComponent()
    BoneName=upperarm_r
    BaseName=clavicle_r
    InputRange=(X=0,Y=0)
    OutputRange=(X=0,Y=6)
    TwistAxis="1,0,0"
   )
   > Set Hand rise(Hand rise:float=float * float!(A=@n10.RawOutput, B=-0.2) -> (ReturnValue)) -> (Output_Get:float)
   n11 = float * float(A=@n10.RawOutput, B=-0.05)
   > Set Headphone_left(Headphone_left:double=FInterpTo(
     Current=Get Headphone_left() -> (Headphone_left:double)
     Target="Clamp (Float)"(Value=@n11, Min=-1, Max=0)
     DeltaTime=@n4.DeltaTimeX
     InterpSpeed=7
    )) -> (Output_Get:double)
   n12 = GetAlphaFromReferenceTransformDelta(
    SkeletalMeshComponent=GetOwnerSkeletalMeshComponent()
    BoneName=head
    BaseName=spine_05
    DeltaComponentType=EulerX
    InputRange=(X=0,Y=0)
    TwistAxis="1,0,0"
   )
   > Set Head rotation X(Head rotation X:float=FInterpTo(
     Current=Get Head rotation X() -> (Head rotation X:float)
     Target="Clamp (Float)"(Value=@n12.RawOutput, Min=0, Max=1.7)
     DeltaTime=@n4.DeltaTimeX
     InterpSpeed=1
    )) -> (Output_Get:float)
   > Set Head rotation X full(Head rotation X full:float=FInterpTo(
     Current=Get Head rotation X full() -> (Head rotation X full:float)
     Target="Clamp (Float)"(Value=@n12.RawOutput, Min=-15, Max=0)
     DeltaTime=@n4.DeltaTimeX
     InterpSpeed=1
    )) -> (Output_Get:float)
   > Set Left_ear(Left_ear:double=float + float(A=Get Headphone_left(), B=Get Head rotation X())) -> (Output_Get:double)
   n13 = float * float(A=@n10.RawOutput, B=0.02)
   > Set Headphone_Right(Headphone_Right:float=FInterpTo(Current=Get Headphone_Right(), Target="Clamp (Float)"(Value=@n13, Min=0, Max=1), DeltaTime=@n4.DeltaTimeX, InterpSpeed=3)) -> (Output_Get:float)
   n14 = Get bIsTargeting()
   n15 = Get bIsMoving2D()
   > Set Heaphone_crouch(Heaphone_crouch:bool=AND Boolean!(A=Get bIsCrouching(), B=NOT Boolean(A=@n15), C:bool=NOT Boolean(A=@n14)) -> (ReturnValue)) -> (Output_Get:bool)
   > n16 = Gravity Override State Function DEPRECATED(CharacterPartAnimInstance=Self-Reference(), GameTimeInSeconds=GetGameTimeInSeconds(), GravityOverrideParams=Get Gravity Override Params())
   > Set Gravity Override(Gravity Override=@n16)
   n17 = Get bIsPlayingMeleeAnim()
   n18 = OR Boolean(A=Get bIsJumping(), B=Get bIsFalling())
   n19 = NOT Boolean(A=@n18)
   n20 = Get bIsMoving2D()
   n21 = float * float(A=Vector Length(A=Get PawnVelocity()), B="To Float (Boolean)"(InBool=AND Boolean!(A=@n20, B=@n19) -> (ReturnValue)))
   > Set rigid_body_alpha(rigid_body_alpha=float * float!(A=SelectFloat(A=0.2, B=0.7, bPickA=@n17), B=MapRangeClamped(Value=@n21, InRangeA=550, InRangeB=770, OutRangeA=0.7, OutRangeB=0.3), C:double) -> (ReturnValue))
   > Set Constraint alpha(Constraint alpha=MapRangeClamped(Value=@n21, InRangeA=30, InRangeB=550, OutRangeB=0.3))
   > Set jump easing(jump easing:double=MapRangeClamped(Value=Get PawnVelocityZ(), InRangeA=630, InRangeB=-630, OutRangeA=15, OutRangeB=-15)) -> (Output_Get:double)
   > Set side jump easing(side jump easing:double=MapRangeClamped(Value=Get PawnVelocityZ(), InRangeA=630, InRangeB=-630, OutRangeA=8, OutRangeB=-8)) -> (Output_Get:double)
   > Set jump easing tr(jump easing tr:double=MapRangeClamped(Value=Get PawnVelocityZ(), InRangeA=630, InRangeB=-630, OutRangeA=2, OutRangeB=0)) -> (Output_Get:double)
   > Set is in jump or falling(is in jump or falling:bool=@n18) -> (Output_Get:bool)
   Is Valid(exec=@n5.then_1, InputObject=Get PlayerPawn() -> (PlayerPawn:Pawn Object Reference))
   > Set WPO Globals Function(
    Material=Get Mat Instance Dynamic()
    Overlay Material=Get Overlay Mat Dynamic() -> (Overlay Mat Dynamic:Material Instance Dynamic Object Reference)
    Skeletal Mesh=Get SkelMesh()
    "Local Noise Transform  Location"="Get Local Noise Transform  Location"() -> ("Local Noise Transform  Location":Vector)
    Local Noise Offset=Get Local Noise Offset()
   )
   > Set WPO Regional Function!(
    execute
    Material:Material Instance Dynamic Object Reference=Get Mat Instance Dynamic()
    Overlay Material:Material Instance Dynamic Object Reference=Get Overlay Mat Dynamic()
    Skeletal Mesh:Skeletal Mesh Component Object Reference=Get SkelMesh()
    Velocity Transform:Name=pelvis
    Velocity Interp Speed:double=Get Velocity Interp Speed()
    Collision Transform:Name=spine_03
    Collision Transform Axis:EAxis Enum
    Invert Collision Transform Axis:bool=false
    Delta Time X:double=@n4.DeltaTimeX
    Previous Velocity Transform Location:"Vector (by ref)"=Get R1_Previous Velocity Transform Location() -> (R1_Previous Velocity Transform Location:Vector)
    Previous Velocity Transform Velocity:"Vector (by ref)"=Get R1_Previous Velocity Transform Velocity() -> (R1_Previous Velocity Transform Velocity:Vector)
    Region:int=1
   ) -> (then)
   > Set WPO Regional Function(
    Material=Get Mat Instance Dynamic()
    Overlay Material=Get Overlay Mat Dynamic()
    Skeletal Mesh=Get SkelMesh()
    Velocity Transform=lowearm_r
    Velocity Interp Speed=Get Velocity Interp Speed()
    Collision Transform=upperarm_r
    Invert Collision Transform Axis=false
    Delta Time X=@n4.DeltaTimeX
    Previous Velocity Transform Location=Get R2_Previous Velocity Transform Location() -> (R2_Previous Velocity Transform Location:Vector)
    Previous Velocity Transform Velocity=Get R2_Previous Velocity Transform Velocity() -> (R2_Previous Velocity Transform Velocity:Vector)
    Region=2
   )
   > SetScalarParameterValue(self=Get Mat Instance Dynamic(), ParameterName=ClothSim_WindStrength, Value=SelectFloat(A=1, B=0, bPickA=Get bIsSkydiveFloating()))
   Event BlueprintInitializeAnimation()
   > Set SkelMesh(SkelMesh=GetOwnerSkeletalMeshComponent())
   n22 = TryGetPawnOwner()
   > Is Valid(InputObject=@n22)
   > Set PlayerPawn(PlayerPawn:Pawn Object Reference=@n22) -> (Output_Get:Pawn Object Reference)
  ]}
]
variables [
 Constraint alpha {type=number}
 Gravity Override {type=Vector}
 Gravity Override Params {type=GravityOverrideParamsStruct}
 Hand rise {type=number}
 Head rotation X {type=number}
 Head rotation X full {type=number}
 Headphone_left {type=number}
 Headphone_Right {type=number}
 Heaphone_crouch {type=boolean}
 is in jump or falling {type=boolean}
 is left foot bends back {type=boolean}
 is right foot bends back {type=boolean}
 jump easing {type=number}
 jump easing tr {type=number}
 Left_ear {type=number}
 Local Noise Offset {type=Vector}
 "Local Noise Transform  Location" {type=Vector}
 Mat Instance Dynamic {type=MaterialInstanceDynamic}
 Overlay Mat Dynamic {type=MaterialInstanceDynamic}
 PlayerPawn {type=Pawn, default=null}
 R1_Previous Velocity Transform Location {type=Vector}
 R1_Previous Velocity Transform Velocity {type=Vector}
 R2_Previous Velocity Transform Location {type=Vector}
 R2_Previous Velocity Transform Velocity {type=Vector}
 R3_Previous Velocity Transform Location {type=Vector}
 R3_Previous Velocity Transform Velocity {type=Vector}
 rigid_body_alpha {type=number}
 shoe back rot l {type=number}
 side jump easing {type=number}
 SkelMesh {type=SkeletalMeshComponent}
 Velocity Interp Speed {type=number, default=1}
]
